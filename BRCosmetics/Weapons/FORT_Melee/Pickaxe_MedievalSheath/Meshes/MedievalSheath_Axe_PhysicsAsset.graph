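# Physics asset for the MedievalSheath axe pickaxe, a two-handed melee cosmetic weapon mesh.
# It holds a single capsule body on the Melee_TwoHanded_Handle bone, rotated to lie along the weapon and offset along it.
# There are no constraints, so the weapon collides and simulates as one rigid piece.
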PhysicsAsset /BRCosmetics/Weapons/FORT_Melee/Pickaxe_MedievalSheath/Meshes/MedievalSheath_Axe_PhysicsAsset
bodies {Melee_TwoHanded_Handle [{shapeType=Capsule, length=76.057, radius=35.3326, center=<0.7422 -0.6636 27.2237>, rotation=<rot 1.7757 -90 0.7251>}]}
constraints []
